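# Animation Blueprint for the SubtleSilly backpack cosmetic, driving its secondary motion and toggling a material element.
# Spine yaw relative to the reference pose feeds additive pack-bone offsets and rotation each update.
# Chained spring-based Anim Dynamics nodes then simulate the left and right loops and laces.
# Weapon, customization, sprint, slide and skydive events set "Hide Element 01" based on the equipped pickaxe's SubtleSilly metatag and movement state.
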AnimBlueprint /BRCosmetics/Accessories/FORT_Backpacks/Backpack_M_MED_SubtleSilly/Meshes/M_MED_SubtleSilly_Pack_AnimBP
target_skeleton=&/M_MED_SubtleSilly_Pack_Skeleton
graphs [
 AnimGraph {nodes [
   n1 = Local To Component(LocalPose=Input Pose())
   n2 = Anim Dynamics - Bone: dyn_loop_l(
    ComponentPose="Transform (Modify) Bone - Bone: pack"(
     ComponentPose="Transform (Modify) Bone - Bone: pack"(ComponentPose=@n1, Translation="0,-1,0", Rotation="0,0,0") {translationMode=BMM_Additive, translationSpace=BCS_BoneSpace, boneToModify {boneName=pack}, translation {y=-1}}
     Translation=Get backpack_translation()
     Rotation=Get backpack_rotate()
    ) {rotationMode=BMM_Additive, translationMode=BMM_Additive, boneToModify {boneName=pack}}
    Alpha=1
    ExternalForce="0,0,0"
   ) {
    angularDampingOverride=1
    angularSpringConstant=4000
    bAngularSpring=true
    bOverrideAngularDamping=true
    bUseGravityOverride=true
    gravityScale=0.35
    simulationSpace=RootRelative
    boundBone {boneName=dyn_loop_l}
    componentAppliedLinearAccClamp=<3000 3000 3000>
    componentLinearAccScale=<0.9 0.9 0.9>
    componentLinearVelScale=<1 1 1>
    gravityOverride {z=-980}
    physicsBodyDefinitions [{
      boundBone {boneName=dyn_loop_l}
      boxExtents=<20 20 20>
      constraintSetup {angularLimitsMax=<10 3 75>, angularLimitsMin=<-10 -75 -15>, angularTarget=<1 0 0>}
      localJointOffset=<3 0 0>
     }]
    simSpaceSettings {maxAngularAcceleration=25, maxAngularVelocity=9, simSpaceAngularAlpha=1}
   }
   n3 = Anim Dynamics - Bone: dyn_loop_r(ComponentPose=@n2, Alpha=1, ExternalForce="0,0,0") {
    angularDampingOverride=1
    angularSpringConstant=4000
    bAngularSpring=true
    bOverrideAngularDamping=true
    bUseGravityOverride=true
    gravityScale=0.35
    simulationSpace=RootRelative
    boundBone {boneName=dyn_loop_r}
    componentAppliedLinearAccClamp=<3000 3000 3000>
    componentLinearAccScale=<0.9 0.9 0.9>
    componentLinearVelScale=<1 1 1>
    gravityOverride {z=-980}
    physicsBodyDefinitions [{
      boundBone {boneName=dyn_loop_r}
      boxExtents=<20 20 20>
      constraintSetup {angularLimitsMax=<10 3 15>, angularLimitsMin=<-10 -80 -80>, angularTarget=<1 0 0>}
      localJointOffset=<3 0 0>
     }]
    simSpaceSettings {maxAngularAcceleration=25, maxAngularVelocity=7, simSpaceAngularAlpha=1}
   }
   n4 = Anim Dynamics - Bone: dyn_lace_l_01(ComponentPose=@n3, Alpha=1, ExternalForce="0,0,0") {
    angularDampingOverride=1
    angularSpringConstant=4000
    bAngularSpring=true
    bOverrideAngularDamping=true
    bUseGravityOverride=true
    gravityScale=0.35
    simulationSpace=RootRelative
    boundBone {boneName=dyn_lace_l_01}
    componentAppliedLinearAccClamp=<3000 3000 3000>
    componentLinearAccScale=<0.9 0.9 0.9>
    componentLinearVelScale=<1 1 1>
    gravityOverride {z=-980}
    physicsBodyDefinitions [{
      boundBone {boneName=dyn_lace_l_01}
      boxExtents=<17 17 17>
      constraintSetup {angularLimitsMax=<5 5 70>, angularLimitsMin=<-5 -70 -5>, angularTarget=<1 0 0>}
      localJointOffset=<3 0 0>
     }]
    simSpaceSettings {maxAngularAcceleration=25, maxAngularVelocity=9, simSpaceAngularAlpha=1}
   }
   n5 = Anim Dynamics - Bone: dyn_lace_l_02(ComponentPose=@n4, Alpha=1, ExternalForce="0,0,0") {
    angularDampingOverride=1
    angularSpringConstant=4000
    bAngularSpring=true
    bOverrideAngularDamping=true
    bUseGravityOverride=true
    gravityScale=0.35
    simulationSpace=RootRelative
    boundBone {boneName=dyn_lace_l_02}
    componentAppliedLinearAccClamp=<3000 3000 3000>
    componentLinearAccScale=<0.9 0.9 0.9>
    componentLinearVelScale=<1 1 1>
    gravityOverride {z=-980}
    physicsBodyDefinitions [{
      boundBone {boneName=dyn_lace_l_02}
      boxExtents=<17 17 17>
      constraintSetup {angularLimitsMax=<5 5 70>, angularLimitsMin=<-5 -70 -5>, angularTarget=<1 0 0>}
      localJointOffset=<3 0 0>
     }]
    simSpaceSettings {maxAngularAcceleration=25, maxAngularVelocity=9, simSpaceAngularAlpha=1}
   }
   n6 = Anim Dynamics - Bone: dyn_lace_r_01(ComponentPose=@n5, Alpha=1, ExternalForce="0,0,0") {
    angularDampingOverride=1
    angularSpringConstant=4000
    bAngularSpring=true
    bOverrideAngularDamping=true
    bUseGravityOverride=true
    gravityScale=0.35
    simulationSpace=RootRelative
    boundBone {boneName=dyn_lace_r_01}
    componentAppliedLinearAccClamp=<3000 3000 3000>
    componentLinearAccScale=<0.9 0.9 0.9>
    componentLinearVelScale=<1 1 1>
    gravityOverride {z=-980}
    physicsBodyDefinitions [{
      boundBone {boneName=dyn_lace_r_01}
      boxExtents=<17 17 17>
      constraintSetup {angularLimitsMax=<5 5 5>, angularLimitsMin=<-5 -60 -60>, angularTarget=<1 0 0>}
      localJointOffset=<3 0 0>
     }]
    simSpaceSettings {maxAngularAcceleration=25, maxAngularVelocity=7, simSpaceAngularAlpha=1}
   }
   n7 = Anim Dynamics - Bone: dyn_lace_r_02(ComponentPose=@n6, Alpha=1, ExternalForce="0,0,0") {
    angularDampingOverride=1
    angularSpringConstant=4000
    bAngularSpring=true
    bOverrideAngularDamping=true
    bUseGravityOverride=true
    gravityScale=0.35
    simulationSpace=RootRelative
    boundBone {boneName=dyn_lace_r_02}
    componentAppliedLinearAccClamp=<3000 3000 3000>
    componentLinearAccScale=<0.9 0.9 0.9>
    componentLinearVelScale=<1 1 1>
    gravityOverride {z=-980}
    physicsBodyDefinitions [{
      boundBone {boneName=dyn_lace_r_02}
      boxExtents=<17 17 17>
      constraintSetup {angularLimitsMax=<5 5 5>, angularLimitsMin=<-5 -60 -60>, angularTarget=<1 0 0>}
      localJointOffset=<3 0 0>
     }]
    simSpaceSettings {maxAngularAcceleration=25, maxAngularVelocity=9, simSpaceAngularAlpha=1}
   }
   Output Pose(Result=Component To Local(ComponentPose=@n7)) {layerGroup=None}
  ]}
 EventGraph {nodes [
   n1 = Event OnBeginTacticalSprinting() {is=title K2Node_Event}
   n2 = Get Dynamic Material Instance(execute=[@n6.then, @n15.then], SkeletalMeshComponent=GetPartSkeletalMeshComponent())
   SetScalarParameterValue(execute=[@n1.then, @n7.then, @n13.then, @n5.else, @n10.else], self=@n2, ParameterName=Hide Element 01, Value=0)
   n3 = Event OnEndTacticalSprinting() {is=title K2Node_Event}
   n4 = PickaxeItemDefinitionHasAnyMetaTag(Tags=Make Array("[0]"=(TagName="Cosmetics.Metatag.SubtleSilly")), bRequiresEquipped=true)
   n5 = Branch(execute=[@n3.then, @n8.then], Condition=@n4)
   SetScalarParameterValue(execute=[@n13.else, @n5.then], self=@n2, ParameterName=Hide Element 01, Value=1)
   n6 = Event OnWeaponChanged()
   n7 = Event OnBeginSkydiving()
   n8 = Event OnEndSkydiving()
   n9 = Event OnBeginSliding() {is=title K2Node_Event}
   n10 = Branch(execute=[@n2.then, @n9.then, @n14.then], Condition=@n4)
   n11 = NOT Boolean(A=Get bIsSliding())
   n12 = Get bIsTacticalSprinting()
   > n13 = Branch(Condition=OR Boolean(A=Get bIsSkyDiving(), B=AND Boolean(A=@n12, B=@n11)))
   n14 = Event OnEndSliding() {is=title K2Node_Event}
   n15 = Event OnCharacterCustomizationLoaded()
   Event BlueprintUpdateAnimation()
   n16 = GetOwnerSkeletalMeshComponent()
   n17 = BreakRotator(InRot=BreakTransform(InTransform=GetDeltaTransformFromRefPose(self=@n16, BoneName=spine_05, BaseName=spine_01)).Rotation)
   > Set backpack_rotate(backpack_rotate=MakeRotator(Roll=MapRangeClamped(Value=@n17.Yaw, InRangeA=0, InRangeB=-30, OutRangeB=20)))
   > Set backpack_translation(backpack_translation=MakeVector(Y=MapRangeClamped(Value=@n17.Yaw, InRangeA=0, InRangeB=-30, OutRangeB=-2)))
   TryGetPawnOwner()
  ]}
]
variables [
 backpack_rotate {type=Rotator}
 backpack_translation {type=Vector}
]
